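# Translucent unlit material for a splash mesh in the Apollo Tower "After" set, with Runtime Virtual Texture output.
# A HeightLerp on the splash mask's red channel blends two tints and varies roughness and specular; inverted blue drives opacity through another HeightLerp.
# The normal map is transformed before feeding the RVT output and a virtual texture switch using EnableRVT.
# A world-position sphere mask branch is left unconnected.
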Material /Game/Environments/Apollo/Sets/Tower/Meshes/After/M_Splash2
graph {
 properties {
  bTangentSpaceNormal=false
  bUseMaterialAttributes=true
  blendMode=BLEND_Translucent
  pixelDepthOffsetMode=PDOM_Legacy
  shadingModel=MSM_Unlit
  usedShadingModels=Unlit
 }
 nodes [
  n1 = WorldPosition()
  n2 = VectorParameter() {parameterName=MaskLocation, defaultValue=<rgba 6284 7012 -4132 0>}
  n3 = ScalarParameter() {defaultValue=7656.5444}
  n4 = ScalarParameter() {defaultValue=0.75, parameterName=Param_1}
  Multiply(A=Saturate(None=SphereMask(A=@n1.XYZ, B=@n2.RGB, Radius=@n3, Hardness=@n4)), B=ScalarParameter() {defaultValue=0.75, parameterName=Param_2})
  n5 = VectorParameter() {parameterName=Param_2, defaultValue=<rgba 0.4445 0.2227 1.5 0>}
  n6 = TextureSample() {samplerType=SAMPLERTYPE_Masks, texture=/Game/Environments/Apollo/Sets/Tower/Meshes/After/Textures/Splash_2_M}
  n7 = VectorParameter() {parameterName=Param_1, defaultValue=<rgba 0.4154 0.0257 0.4297 0>}
  n8 = MaterialFunctionCall(
   Transition Phase=ScalarParameter() {defaultValue=0.65, parameterName=Param_3}
   Height Texture=@n6.R
   Contrast=ScalarParameter() {defaultValue=1.8062, parameterName=ss}
  ) {materialFunction=&HeightLerp}
  n9 = SetMaterialAttributes!(
   MaterialAttributes
   Base Color=LinearInterpolate(
    A=Multiply(A=@n5.RGB, B=@n6.R)
    B=MaterialFunctionCall(Base=@n7.RGB, Blend=@n6.R) {materialFunction=&Blend_Overlay}.Result
    Alpha=@n8.Alpha
   )
   Roughness=ScalarParameter() {defaultValue=0.45, parameterName=Param_7}
   Specular=ScalarParameter() {defaultValue=0.1, parameterName=Param_8}
   Normal=Multiply(
    A=TextureSample() {samplerType=SAMPLERTYPE_Normal, texture=/Game/Environments/Apollo/Sets/Tower/Meshes/After/Textures/Splash_2_N}.RGB
    B=VectorParameter() {defaultValue=<rgba 1 1 1 0>}.RGB
   )
  ) {attributeSetTypes [
    #1
    #2
    #3
    #4
   ]}
  n10 = SetMaterialAttributes!(
   MaterialAttributes=@n9
   Opacity=MaterialFunctionCall(
    Transition Phase=ScalarParameter() {defaultValue=0.5, parameterName=Param_4}
    Height Texture=OneMinus(None=@n6.B)
    Contrast=ScalarParameter() {defaultValue=0.65, parameterName=Param_5}
   ) {materialFunction=&HeightLerp}.Alpha
  ) {attributeSetTypes [#5]}
  n11 = GetMaterialAttributes(None=@n10) {attributeGetTypes [#3, #2]}
  n12 = Multiply(A=@n11.Specular) {constB=0.35}
  n13 = Constant()
  n14 = GetMaterialAttributes(None=BlendMaterialAttributes(
    A=SetMaterialAttributes!(MaterialAttributes=@n11.MaterialAttributes, Specular=@n12, Roughness=@n13) {attributeSetTypes [#3, #2]}
    B=@n10
    Alpha=@n8.Alpha
   )) {attributeGetTypes [#4]}
  n15 = SetMaterialAttributes!(MaterialAttributes=@n14.MaterialAttributes, Normal=Transform(None=@n14.Normal)) {attributeSetTypes [#4]}
  n16 = GetMaterialAttributes(None=@n15) {attributeGetTypes [
    #1
    #3
    #2
    #4
    #5
   ]}
  RuntimeVirtualTextureOutput!(
   BaseColor=@n16.Base Color
   Specular=@n16.Base Color
   Roughness=@n16.Base Color
   Normal=@n16.Base Color
   WorldHeight
   Opacity=@n16.Base Color
   Mask
   Displacement
   Mask4
  )
  n17 = VirtualTextureFeatureSwitch!(
   No=@n15
   Yes=MaterialFunctionCall!(RVT=@n16.MaterialAttributes, ByPass RVT=@n15) {materialFunction=/Game/Athena/Apollo/Maps/Landscape/VT/MaterialFunctions/EnableRVT}.Result
  )
  ShadingPathSwitch(Default=@n17, Deferred=@n17, Forward=@n17, Mobile=@n17)
 ]
}
